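; Server-side global manager for CherrySmoke hero ranking that awards hero points to teams for surviving storm phases.
; On BeginPlay it caches scalable-float tuning and listens to the global safe zone phase event until EndPlay.
; Phases after the first circle raise the stored award cumulatively within a configured circle range.
; A set of teams built from the GameState player array gives each squad the points once per phase.
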(event EventBeginPlay
  (Networking|SwitchHasAuthority
    (:Authority
      (CallFunction|ListenforGlobalStormEvent))))

(event EventEndPlay
  (params (EndPlayReason :type EEndPlayReason%20Enum :default "Destroyed"))
  (Networking|SwitchHasAuthority
    (:Authority
      (GameplayEventRouter|StopListeningtoStatefulEvent (Variables|Default|GetSafeZoneEventListeningTask)))))

(event Custom|ListenForGlobalStormEvent
  ; The Storm Survived event is a global event so we want to make sure each squad only gets it once
  (CallFunction|CacheScalableFloats)
  (bind (_asynctaskproxy _payload) (GameplayEventRouter|StartListeningToStatefulEvent (GameplayEventRouter|GetGlobalEventRouter) "/Script/FortniteGame.FortSafeZonePhaseUpdatedEvent")
    (:then
      (Variables|Default|SetSafeZoneEventListeningTask _asynctaskproxy))
    (:OnEventReceived
      _)
    (:OnSavedEventStateExists
      _))
  ;;[0.8229 0.8229 0.8229 0.5] We skip first circle formed, which counts as a 'circle closed event'
  (if (> (Utilities|Struct|BreakFortSafeZonePhaseUpdatedEvent _payload) 1)
    (bind _survivestormcirclepoints (CallFunction|AdjustPointsBasedonCircle (Utilities|Struct|BreakFortSafeZonePhaseUpdatedEvent _payload)))
    (bind _teamsset1 (CallFunction|GetTeamsLeft))
    (bind _valuepin (Utilities|Set|ForEachLoop(Set) _teamsset1))
    (CallFunction|AddHeroPointstoTeam _valuepin _survivestormcirclepoints (GameplayTags|MakeLiteralGameplayTag "(TagName=\"CherrySmoke.HeroRank.Points.Context.SurviveStormPhase\")"))))

(fn AdjustPointsBasedOnCircle
  (params (Safe%20Zone%20Phase%20Index :type Integer))
  (if (and (>= Safe%20Zone%20Phase%20Index (Variables|Default|GetSurviveStormCirclePointsIncreaseMinCircle)) (<= Safe%20Zone%20Phase%20Index (Variables|Default|GetSurviveStormCirclePointsIncreaseMaxCircle)))
    (bind _output_get (Variables|Default|SetSurviveStormCirclePoints (+ (Variables|Default|GetSurviveStormCirclePoints) (Variables|Default|GetSurviveStormCirclePointsIncrease))))
    (return _output_get)
    (else
      (return (Variables|Default|GetSurviveStormCirclePoints)))))

(fn CacheScalableFloats
  (Variables|Default|SetSurviveStormCirclePoints (ScalableFloat|GetValueatLevel (Variables|HF|GetHF_SurviveStormCirclePoints)))
  (Variables|Default|SetSurviveStormCirclePointsIncrease (ScalableFloat|GetValueasInteger (Variables|HF|GetHF_SurviveStormCirclePointsIncrease)))
  (Variables|Default|SetSurviveStormCirclePointsIncreaseMinCircle (ScalableFloat|GetValueasInteger (Variables|HF|GetHF_SurviveStormCirclePointsIncreaseMinCircle)))
  (Variables|Default|SetSurviveStormCirclePointsIncreaseMaxCircle (ScalableFloat|GetValueasInteger (Variables|HF|GetHF_SurviveStormCirclePointsIncreaseMaxCircle))))

(fn GetTeamsLeft
  (for _array_element (Variables|/Script/Engine.GameStateBase|GetPlayerArray (Fortnite|Athena|GameState|GetGameStateAthena))
    (bind _asfort_player_state_athena (Utilities|Casting|CastToFortPlayerStateAthena _array_element))
    (Utilities|Set|Add (LocalVariables|GetTeamsLeft) (Class|FortPlayerStateAthena|GetTeam _asfort_player_state_athena)))
  (return (LocalVariables|GetTeamsLeft)))
