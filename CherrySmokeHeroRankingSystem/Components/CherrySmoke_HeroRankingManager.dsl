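; Player-state component that manages Hero Rank, mirroring it as loose gameplay tags on the owner and applying a removable first-use benefit effect at BeginPlay.
; On rank change it updates tags, sets the rank FX value, optionally heals the pawn and sends a quest gameplay event.
; The authority also runs an EQS query for nearby discoverables, then marks results after a delay, optionally clearing the previous marker.
; ScalableFloat hotfixes control healing, marker clearing and delay.
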(event EventOnRuntimeDataCreated
  (CallFunction|UpdateRankTag))

(event EventBeginPlay
  (CallFunction|UpdateRankTag)
  ; Apply "First Usage Benefit" to owning player. This will be removed externally once we use the Hero Ranking for the first time so this benefit is removed
  (GameplayEffects|ApplyGameplayEffectToSelf (Ability|GetFortAbilitySystemComponent (Components|GetOwner)) "/CherrySmokeHeroRankingSystem/Gameplay/GameplayEffects/GE_CherrySmokeHeroRank_FirstUseBenefit.GE_CherrySmokeHeroRank_FirstUseBenefit_C" 0.0 (Class|AthenaFunctionLibrary|MakeSimpleFortEffectContext (Components|GetOwner) (Components|GetOwner)))
  (Variables|Default|SetHealOnRankUp (ScalableFloat|GetValueasBool (Variables|Hotfixes|GetHF_HealOnRankUp)))
  (Variables|Default|SetShouldRemovePreviousChestPing (ScalableFloat|GetValueasBool (Variables|Hotfixes|GetHF_RemovePreviousChestPing))))

(event EventOnRankChanged
  (params (NewRank :type Gameplay%20Tag%20Structure%20%28by%20ref%29))
  (CallFunction|UpdateRankTag)
  (Utilities|FlowControl|Sequence
    (:then_0
      (CallFunction|StartGiveRankRewards)
      (CallFunction|SendQuestEventonRankUp))
    (:then_1
      ; Search for closest discoverable
      (Utilities|SwitchHasAuthority
        (:Authority
          (bind _returnvalue (AI|EQS|RunEQSQuery "/CherrySmokeHeroRankingSystem/Gameplay/EQS_HeroRankingDiscoverables.EQS_HeroRankingDiscoverables" (PlayerState|GetPawn (Utilities|Casting|CastToFortPlayerStateAthena (Components|GetOwner)))))
          (AI|EQS|AssignOnQueryFinished _returnvalue (EventDispatchers|CreateEvent :!event OnHEroRankDisoverableEQSFinished)))))))

(event Custom|OnHeroRankDisoverableEQSFinished
  (params (QueryInstance :type EQS%20Query%20Instance%20Object%20Reference) (QueryStatus :type EEnvQueryStatus%20Enum :default "Processing"))
  ; Mark the Closest Chest
  (Variables|Default|SetQuery%20Instance QueryInstance)
  (bind _returnvalue (Utilities|Time|SetTimerbyEvent (EventDispatchers|CreateEvent :!event Custom|PingChestDelay) (ScalableFloat|GetValueatLevel (Variables|Hotfixes|GetHF_PIngChestDelay))))
  (Variables|PingHeroChest|SetPingChestDelayTimerHandle _returnvalue))

(event Custom|PingChestDelay
  ; Mark the Closest Chest
  (Utilities|FlowControl|Sequence
    (:then_0
      (Variables|Default|GetShouldRemovePreviousChestPing
        (:then
          (bind _markedactor (Variables|PingHeroChest|GetMarkedActor)
            (:then
              (MarkerComponent|DeactivateMarkersonActor _markedactor))
            (:else)))
        (:else)))
    (:then_1
      (bind _resultactors (AI|EQS|GetQueryResultsasActors (Variables|Default|GetQuery%20Instance)))
      (for _array_element _resultactors
        (Variables|PingHeroChest|SetMarkedActor _array_element)
        (MarkerComponent|MarkActoronServer (Actor|GetComponentbyClass (Fortnite|GetFortPlayerControllerfromActor (Components|GetOwner)) "/Script/FortniteGame.AthenaMarkerComponent") _array_element false true)
        (Class|BPICherrySmokeTagUnlock|SetMinimapIcontoShow(Message) _array_element false)))))

(event EventOnHeroPointsChanged
  (params (CurrentPoints :type Float%20%28single-precision%29)))

(fn GetHeroRankingTagsFromOwner
  (return (GameplayTags|Filter (GameplayTags|GetOwnedGameplayTags (Ability|GetAbilitySystemComponent (Components|GetOwner)) :!class /Script/GameplayTags.BlueprintGameplayTagLibrary) (GameplayTags|MakeLiteralGameplayTagContainer "(GameplayTags=((TagName=\"CherrySmoke.HeroRank.A\"),(TagName=\"CherrySmoke.HeroRank.B\"),(TagName=\"CherrySmoke.HeroRank.C\"),(TagName=\"CherrySmoke.HeroRank.S\"),(TagName=\"CherrySmoke.HeroRank.S+\")))"))))

(fn RankUpHeal
  (bind _as_fort_player_pawn_athena (Variables|Default|GetAs%20Fort%20Player%20Pawn%20Athena))
  (Ability|Advanced|GrantandActivateAbilityonSelf (Ability|GetAbilitySystemComponent _as_fort_player_pawn_athena) (Variables|Rewards|GetHeal%20Ability%20to%20Grant))
  (GameplayCue|ExecuteGameplayCueOnActor(Burst) _as_fort_player_pawn_athena (Variables|GameplayCues|GetHeal%20Gameplay%20Cue%20Tag) (Ability|GameplayCue|MakeGameplayCueParameters 0.0)))

(fn SendQuestEventOnRankUp
  (Ability|SendGameplayEventtoActor (Variables|Default|GetAs%20Fort%20Player%20Pawn%20Athena) "(TagName=\"CherrySmoke.HeroRank.Event.RankUp\")" (Utilities|Struct|MakeGameplayEventData "(TagName=\"CherrySmoke.HeroRank.Event.RankUp\")" (Variables|Default|GetAs%20Fort%20Player%20Pawn%20Athena))))

(fn StartGiveRankRewards
  (bind _asfort_player_state (Utilities|Casting|CastToFortPlayerState (Components|GetOwner)))
  (bind _asfort_player_pawn_athena (Utilities|Casting|CastToFortPlayerPawnAthena (PlayerState|GetPawn _asfort_player_state)))
  (Variables|Default|SetAs%20Fort%20Player%20Pawn%20Athena _asfort_player_pawn_athena)
  (CallFunction|StartRankUpCosmetics)
  (Variables|Default|GetHealOnRankUp
    (:then
      (CallFunction|RankUpHeal))
    (:else)))

(fn StartRankUpCosmetics
  (switch Utilities|FlowControl|Switch|SwitchonGameplayTag (CallFunction|GetHeroRank)
    (:Default
      _)
    (:CherrySmoke.HeroRank.B
      (Variables|Default|SetCurrentRankFX 0.0))
    (:CherrySmoke.HeroRank.A
      (Variables|Default|SetCurrentRankFX 1.0))
    (:CherrySmoke.HeroRank.S
      _)
    (:CherrySmoke.HeroRank.S+
      (Variables|Default|SetCurrentRankFX 3.0)))
  (Variables|Default|SetCurrentRankFX 2.0))

(fn UpdateRankTag
  (bind _returnvalue (Components|GetOwner))
  ; Updating Loose Tags on Server and Client so that they match the newest rank
  (LocalVariables|SetOwnerHeroRankTags (CallFunction|GetHeroRankingTagsfromOwner))
  (LocalVariables|SetCurrentHeroRankTags (CallFunction|GetHeroRank))
  (if (not (GameplayTags|MatchesAnyTags (LocalVariables|GetCurrentHeroRankTags) (LocalVariables|GetOwnerHeroRankTags) true))
    (Ability|GameplayEffect|RemoveLooseGameplayTags _returnvalue (LocalVariables|GetOwnerHeroRankTags))
    (Ability|GameplayEffect|AddLooseGameplayTags _returnvalue (GameplayTags|MakeGameplayTagContainerfromTag (LocalVariables|GetCurrentHeroRankTags)))))

(fn UseFirstUseBenefit
  (GameplayEffects|RemoveActiveGameplayEffectbySourceEffect (Ability|GetFortAbilitySystemComponent (Components|GetOwner)) "/CherrySmokeHeroRankingSystem/Gameplay/GameplayEffects/GE_CherrySmokeHeroRank_FirstUseBenefit.GE_CherrySmokeHeroRank_FirstUseBenefit_C" (Ability|GetFortAbilitySystemComponent (Components|GetOwner)) -1))
